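# Material function that bends and stretches a mesh, by name for Niagara, so its RestAxis turns toward FollowPos around MeshRootPos (particle position by default).
# The angle between rest and follow directions, scaled by Rigidity and Mask, drives RotateAboutAxis on world position offset along RestAxis by masked StretchDistance.
# It outputs the resulting WPO, an equally rotated normal, and mask-blended RotateVector tangents U/V for SplineThicken.
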MaterialFunction /Game/Effects/Niagara/Modules/FollowingDeform/Material/MF_NiagaraFollowingDeform
graph {
 properties {bExposeToLibrary=true, libraryCategoriesText [FollowingDeform]}
 nodes [
  n1 = Normalize(VectorInput=FunctionInput() {inputName=RestAxis})
  n2 = FunctionInput() {inputName=FollowPos}
  n3 = Constant3Vector() {constant=<rgba 0 0 1 0>}
  n4 = FunctionInput(Preview=ParticlePositionWS()) {bUsePreviewValueAsDefault=true, description=Default is particle position, inputName=MeshRootPos}
  n5 = Normalize(VectorInput=Subtract(A=Add(A=@n2, B=@n3), B=@n4))
  n6 = CrossProduct(A=@n1, B=@n5)
  n7 = DotProduct(A=@n5, B=@n1)
  n8 = FunctionInput() {bUsePreviewValueAsDefault=true, inputName=Rigidity, inputType=FunctionInput_Scalar, sortPriority=10, previewValue=<0.1592 0 0 1>}
  n9 = FunctionInput() {inputName=Mask, inputType=FunctionInput_Scalar, sortPriority=-1}
  n10 = Multiply(A=@n9, B=Multiply(A=ArccosineFast!(None=@n7), B=@n8))
  n11 = Multiply(A=Multiply(A=@n1, B=FunctionInput() {bUsePreviewValueAsDefault=true, inputName=StretchDistance, inputType=FunctionInput_Scalar}), B=@n9)
  n12 = Add(A=@n11, B=WorldPosition() {worldPositionShaderOffset=WPT_ExcludeAllShaderOffsets}.XYZ)
  FunctionOutput(None=Add(A=RotateAboutAxis(NormalizedRotationAxis=@n6, RotationAngle=@n10, PivotPoint=@n4, Position=@n12), B=@n11)) {outputName=WPO}
  n13 = Normalize(VectorInput=Transform(None=Constant3Vector() {constant=<rgba 1 0 0 0>}))
  FunctionOutput(None=LinearInterpolate(
    A=@n13
    B=MaterialFunctionCall!(VectorToRotate=@n13, LookAtVector=@n5, RestingVector=@n1) {materialFunction=/Engine/Functions/Engine_MaterialFunctions01/Vectors/RotateVector}.Result
    Alpha=@n9
   )) {outputName=ForSplineThicken_TangentU_WS, sortPriority=11}
  n14 = Normalize(VectorInput=Transform(None=Constant3Vector() {constant=<rgba 0 1 0 0>}))
  FunctionOutput(None=LinearInterpolate(
    A=@n14
    B=MaterialFunctionCall(VectorToRotate=@n14, LookAtVector=@n5, RestingVector=@n1) {materialFunction=&RotateVector}.Result
    Alpha=@n9
   )) {outputName=ForSplineThicken_TangentV_WS, sortPriority=12}
  n15 = Constant()
  n16 = FunctionInput(Preview=Transform(None=Constant3Vector() {constant=<rgba 0 0 1 0>})) {bUsePreviewValueAsDefault=true, description=Default is pixel normal, inputName=NormalWS}
  FunctionOutput(None=Add(A=RotateAboutAxis(NormalizedRotationAxis=@n6, RotationAngle=@n10, PivotPoint=@n15, Position=@n16), B=@n16)) {outputName=NormalWS, sortPriority=1}
 ]
}
