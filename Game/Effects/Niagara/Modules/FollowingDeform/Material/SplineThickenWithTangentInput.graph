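# Material function that widens very thin polygon strips, such as hair, grass or cables, into camera-facing ribbons with rounded shading.
# World position offset follows the cross product of view vector and tangent (overridable input, TangentBasis default), scaled by a tip-to-base width lerp.
# A hemispherical normal derived from UVs blends with an optional extra normal and rotates into that ribbon basis.
# Screen-space UV-to-world derivatives scale a parallax shift for the texture UVs.
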MaterialFunction /Game/Effects/Niagara/Modules/FollowingDeform/Material/SplineThickenWithTangentInput
graph {
 properties {
  bExposeToLibrary=true
  description="This function  widens strips of polies along the U of their UVs.  These strips shouuld be extremely thin and imported with remove degenerate triangles disabled. Useful for Hair, Grass, Cables etc."
 }
 nodes [
  n1 = FunctionInput(Preview=TextureCoordinate()) {
   bUsePreviewValueAsDefault=true
   description=The UV channel used for texturing. This must be the same UV channel you want to have the 3d-parallax correction.
   inputName=UVs for texturing
   inputType=FunctionInput_Vector2
   sortPriority=5
  }
  n2 = FunctionInput(Preview=StaticBool() {value=true}) {
   bUsePreviewValueAsDefault=true
   description="Defaults to True (expands U)"
   inputName=Expand U or V UV Channel
   inputType=FunctionInput_StaticBool
   sortPriority=6
  }
  n3 = StaticSwitch(True=ComponentMask(None=@n1) {g=true}, False=ComponentMask(None=@n1) {r=true}, Value=@n2)
  n4 = Abs(None=DDX(Value=@n3))
  n5 = FunctionInput(Preview=WorldPosition().XYZ) {bUsePreviewValueAsDefault=true, inputName=WorldPosition, sortPriority=2}
  n6 = MaterialFunctionCall(Vector 3=DDX(Value=@n5)) {materialFunction=/Engine/Functions/Engine_MaterialFunctions02/Utility/VectorLength}
  n7 = Abs(None=DDY(Value=@n3))
  n8 = MaterialFunctionCall(Vector 3=DDY(Value=@n5)) {materialFunction=&VectorLength}
  n9 = Subtract(A=CameraPositionWS(), B=@n5)
  n10 = MaterialFunctionCall() {materialFunction=&TangentBasis}
  n11 = StaticSwitch(
   True=FunctionInput(Preview=@n10.Y V) {bUsePreviewValueAsDefault=true, inputName=TangentV_WS, sortPriority=-11}
   False=FunctionInput(Preview=@n10.X U) {bUsePreviewValueAsDefault=true, inputName=TangentU_WS, sortPriority=-12}
   Value=@n2
  )
  n12 = CrossProduct(A=@n9, B=@n11)
  n13 = MaterialFunctionCall!(Vector1=@n12, Vector2=Multiply(A=@n11, B=Constant() {r=-1})) {materialFunction=/Engine/Functions/Engine_MaterialFunctions02/Math/CreateThirdOrthogonalVector}
  n14 = DotProduct(A=Normalize(VectorInput=@n9), B=@n13.Vector2) {desc=a / h}
  n15 = SquareRoot(None=Subtract(A=Constant() {desc=Hypotenuse, r=1}, B=Multiply(A=@n14, B=@n14)))
  n16 = TextureCoordinate()
  n17 = Frac(None=ComponentMask(None=FunctionInput(Preview=@n16) {bUsePreviewValueAsDefault=true, inputName=UVs For NormalShading, inputType=FunctionInput_Vector2, sortPriority=16}) {r=true})
  n18 = StaticBool()
  n19 = AppendVector(
   A=StaticSwitch(
    True=Constant()
    False=ConstantBiasScale(None=@n17) {bias=-0.5, scale=2}
    Value=FunctionInput(Preview=@n18) {bUsePreviewValueAsDefault=true, inputName=FlattenPixelNormal, inputType=FunctionInput_StaticBool, sortPriority=10}
   )
   B=Constant()
  )
  n20 = DeriveNormalZ(InXY=@n19)
  n21 = ComponentMask(None=@n20) {b=true}
  n22 = TextureCoordinate()
  n23 = FunctionInput(Preview=@n22) {
   bUsePreviewValueAsDefault=true
   description="Texture coordinate to use for Thickness mapping. If projecting from U, it will use the Y component of the specified UV index for the thickness blending. "
   inputName=UVs for Thickness
   inputType=FunctionInput_Vector2
   sortPriority=4
  }
  n24 = LinearInterpolate(
   A=FunctionInput() {inputName=WidthTip, inputType=FunctionInput_Scalar, sortPriority=1, previewValue=<5 0 0 1>}
   B=FunctionInput() {inputName=WidthBase, inputType=FunctionInput_Scalar, previewValue=<5 0 0 1>}
   Alpha=StaticSwitch(True=ComponentMask(None=@n23) {g=true}, False=ComponentMask(None=@n23) {r=true}, Value=@n2)
  )
  n25 = Multiply(A=@n14, B=Multiply(A=Divide(B=@n15), B=Multiply(A=@n21, B=@n24) {desc=Height}))
  n26 = Constant2Vector() {g=1}
  n27 = Constant2Vector() {r=1}
  n28 = Multiply(
   A=Max(A=Divide(A=@n4, B=@n6.V3 Length), B=Divide(A=@n7, B=@n8.V3 Length))
   B=Multiply(A=Multiply(A=@n25) {constB=-1}, B=StaticSwitch(True=@n26, False=@n27, Value=@n2))
  )
  FunctionOutput(None=Add(A=@n1, B=Multiply(A=@n28) {constB=0.5})) {outputName=UVs with Parallax, sortPriority=2}
  n29 = FunctionInput(Preview=StaticBool()) {
   bUsePreviewValueAsDefault=true
   description="Uses DeriveNormalZ to make a perfectly round normal map. Sets the Z height to 0.62 if False (that's the average of half a sphere). If using CustomUVs, this requires a row of verts to be in the center of the mesh, otherwise it will solve z=0 for the whole surface."
   inputName=DeriveNormalZ
   inputType=FunctionInput_StaticBool
   sortPriority=7
  }
  n30 = StaticSwitch(True=@n20, False=AppendVector(A=@n19, B=Constant() {r=0.62}), Value=@n29)
  n31 = FunctionInput() {
   bUsePreviewValueAsDefault=true
   description=Adds a texture normal to the normalmap transform
   inputName=AdditionalNormal
   sortPriority=8
   previewValue=<0 0 1 1>
  }
  n32 = Constant3Vector() {constant=<rgba 1 -1 1 0>}
  n33 = Constant3Vector() {constant=<rgba 1 -1 0 0>}
  n34 = FunctionInput(Preview=StaticBool()) {
   bUsePreviewValueAsDefault=true
   description="Uses DeriveNormalZ to make a perfectly round normal map. Requires a center tesselation vertex otherwise the vale of Z will be 0 for the entire face, causing harsh lighting. When false, Z height is set to 0.62 which approximates the average thickness of a hemisphere."
   inputName=AngleCorrectedNormals
   inputType=FunctionInput_StaticBool
   sortPriority=8
  }
  n35 = StaticSwitch(
   True=MaterialFunctionCall(BaseNormal=@n30, AdditionalNormal=Multiply(A=@n31, B=@n32)) {materialFunction=&BlendAngleCorrectedNormals}.Result
   False=Add(A=Multiply(A=@n31, B=@n33), B=@n30)
   Value=@n34
  )
  n36 = MaterialFunctionCall(Vector1=Multiply(A=@n12) {constB=-1}, Vector2=@n11) {materialFunction=&CreateThirdOrthogonalVector}
  FunctionOutput(None=StaticSwitch(
    True=MaterialFunctionCall(VectorToTransform=@n35, BasisX=@n13.Vector1, BasisY=@n13.Vector1, BasisZ=@n13.Vector1) {materialFunction=&Transform3x3Matrix}.Transformed Vector
    False=MaterialFunctionCall(VectorToTransform=@n35, BasisX=@n36.Vector2, BasisY=@n36.Vector2, BasisZ=@n36.Vector2) {materialFunction=&Transform3x3Matrix}.Transformed Vector
    Value=@n2
   )) {description=Add this result to your other WPO calculations, outputName=Normal}
  n37 = FunctionInput(Preview=@n22) {
   bUsePreviewValueAsDefault=true
   description=The texture coordinate to use in the expansion of the spline. This defaults to texture coordinate 0.
   inputName=UVs for Projection
   inputType=FunctionInput_Vector2
   sortPriority=3
   previewValue=<5 0 0 1>
  }
  n38 = Add(A=Constant() {r=-0.5}, B=@n37)
  n39 = StaticSwitch(True=ComponentMask(None=@n38) {r=true}, False=ComponentMask(None=@n38) {g=true}, Value=@n2)
  FunctionOutput(None=Multiply(A=Normalize(VectorInput=@n12), B=Multiply(A=@n24, B=@n39))) {description=Add this result to your other WPO calculations, outputName=WorldPositionOffset, sortPriority=1}
  If(A=@n31)
  Multiply()
  Multiply(A=@n11, B=@n25) {desc=Push result along the forward vector}
  Normalize(VectorInput=@n13.Vector1)
  Normalize(VectorInput=@n13.Vector3)
  Normalize(VectorInput=@n36.Vector1)
  Normalize(VectorInput=@n36.Vector2)
  Normalize(VectorInput=@n36.Vector3)
  Normalize(VectorInput=@n13.Vector2)
  WorldPosition()
 ]
}
