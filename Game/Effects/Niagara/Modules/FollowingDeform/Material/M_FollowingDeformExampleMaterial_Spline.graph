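# Example lit material for Niagara mesh particles that combines following deformation with spline thickening.
# Dynamic parameters supply follow position, rest axis and stretch distance to MF_NiagaraFollowingDeform, masked by inverted U.
# Its offset position and U tangent (also wired as V) feed SplineThickenWithTangentInput at constant width, yielding reroutes labelled WPO (both offsets summed) and Normal.
# Both appended vectors repeat X in place of Y.
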Material /Game/Effects/Niagara/Modules/FollowingDeform/Material/M_FollowingDeformExampleMaterial_Spline
graph {
 properties {bUsedWithNiagaraMeshParticles=true, pixelDepthOffsetMode=PDOM_Legacy, usedShadingModels=Default Lit}
 nodes [
  n1 = TextureCoordinate()
  n2 = DynamicParameter() {parameterIndex=2, paramNames [FollowPos_X, FollowPos_Y, FollowPos_Z, StretchDistance]}
  n3 = DynamicParameter() {parameterIndex=3, paramNames [RestAxis_X,RestAxis_Y,RestAxis_Z,Param4]}
  n4 = MaterialFunctionCall!(
   Mask=OneMinus(None=MaterialFunctionCall(Float2=@n1) {materialFunction=&BreakOutFloat2Components}.R)
   FollowPos=AppendVector(A=AppendVector(A=@n2.FollowPos_X, B=@n2.FollowPos_X), B=@n2.FollowPos_Z)
   RestAxis=AppendVector(A=AppendVector(A=@n3.RestAxis_X, B=@n3.RestAxis_X), B=@n3.RestAxis_Z)
   MeshRootPos
   NormalWS
   StretchDistance=@n2.StretchDistance
   Rigidity
  ) {materialFunction=&/MF_NiagaraFollowingDeform}
  n5 = Constant() {r=100}
  n6 = MaterialFunctionCall!(
   TangentU_WS=@n4.ForSplineThicken_TangentU_WS
   TangentV_WS=@n4.ForSplineThicken_TangentU_WS
   WidthBase=@n5
   WidthTip=@n5
   WorldPosition=Add(A=@n4.WPO, B=WorldPosition() {worldPositionShaderOffset=WPT_ExcludeAllShaderOffsets}.XYZ)
   UVs for Projection
   UVs for Thickness
   UVs for texturing
   Expand U or V UV Channel=StaticBool()
   DeriveNormalZ
   AdditionalNormal
   AngleCorrectedNormals
   FlattenPixelNormal
   UVs For NormalShading
  ) {materialFunction=&/SplineThickenWithTangentInput}
  Reroute(None=Add(A=@n4.WPO, B=@n6.WorldPositionOffset)) {desc=WPO}
  Reroute(None=@n6.Normal) {desc=Normal}
 ]
}
